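# Sound cue for the emote audio of the BandSoda quadruped cosmetic companion (Mimosa).
# A single wave player feeds the output directly with one emote wave, without randomization or mixing.
# Playback uses shared Mimosa emote properties, Athena emote attenuation and foley concurrency, and the creature pets sound class.
# The cue has boosted volume and virtualizes as play-when-silent, so it keeps its timing when inaudible.
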SoundCue /CosmeticCompanions/Assets/Quadruped/BandSoda/Audio/Emote/SC_Mimosa_BandSoda_Emote
root=WavePlayer_0
nodes [{id=WavePlayer_0, properties {soundWaveAssetPtr=&/Mimosa_BandSoda_Emote_01}}]
properties {
 attenuationSettings=&Fort_Emotes_Athena_Att
 audioPropertiesSheet=/CosmeticCompanions/Assets/Audio_Shared/PropertySheets/APS_Cosmetic_Mimosa_Emote
 maxDistance=1600
 soundClassObject=&Fort_Creatures_Pets
 virtualizationMode=PlayWhenSilent
 volumeMultiplier=1.5
 concurrencySet [/Game/Sounds/Concurrency_Rules/Athena_EmoteFoley_Conc]
}
